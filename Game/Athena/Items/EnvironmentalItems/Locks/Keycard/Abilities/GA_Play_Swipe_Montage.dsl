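; GA_Play_Swipe_Montage is a keycard lock gameplay ability that plays the player's DoorOpen montage, apparently as the card swipe animation.
; On activation it starts a PlayMontageAndWait task, then ends the ability immediately without waiting for the montage.
; Stop-when-ability-ends is off, so the animation keeps playing after the ability finishes.
; The task's callbacks and the end-ability event are empty.
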(event Ability|EventOnEndAbility
  (params (bWasCancelled :type Boolean)))

(event Ability|EventActivateAbility
  (Ability|Tasks|PlayMontageAndWait "None" "/Game/Animation/Game/MainPlayer/Montages/General/DoorOpen.DoorOpen" 1.0 "None" false
    (:OnCompleted
      _)
    (:OnBlendOut
      _)
    (:OnInterrupted
      _)
    (:OnCancelled
      _))
  (Ability|EndAbility))
